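# Niagara system for steam at the Apollo power plant reactor, built from a single sprite emitter named SmokeBlackClose.
# It slowly spawns large, faint sprites across a wide box offset, each drifting upward for a few seconds with sub-UV animation and a slight camera offset.
# Curves fade alpha in and out while scaling RGB from an overbright magenta-purple down to black, tinting the steam as a fading glow.
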NiagaraSystem /Game/Athena/Effects/Particles/Environments/Apollo/NS_Apollo_PowerPlant_Reactor_Steam_Only
dependencies {
 usedDataInterfaces [
  &NiagaraDataInterfaceSpriteRendererInfo
  &NiagaraDataInterfaceVectorCurve
  &NiagaraDataInterfaceCurve
 ]
 usedDynamicInputs []
 usedModules [&SystemState]
}
emitters {SmokeBlackClose {
  emitterUpdateScript [
   &EmitterState(Life Cycle Mode=::NewEnumerator1, Loop Duration=1, UseLoopDelay=true)
   &SpawnRate(SpawnRate=2, Use Spawn Probability=false)
  ]
  particleSpawnScript [
   &InitializeParticle(
    Color=<rgba 0.75 0.75 0.75 0.05>
    Lifetime=&RandomRangeFloat(Minimum=4, Maximum=5, Evaluation Type=::NewEnumerator1)
    Sprite Rotation Angle=&RandomRangeFloat(Minimum=-0.1, Maximum=0.1, Evaluation Type=::NewEnumerator1)
    Sprite Rotation Mode=::NewEnumerator1
    Sprite Size=&Vector2DFromFloat(Value=&RandomRangeFloat(Minimum=2000, Maximum=2500, Evaluation Type=::NewEnumerator1))
    Sprite Size Mode=::NewEnumerator4
    UsePositionOffset=false
    Write Material Random=true
   )
   &Color(Color=&MakeLinearColorFromVectorAndFloat("Vector (RGB)"=<0.75 0.75 0.75>, "Float (Alpha)"=0.05)) {enabled=false}
   &DynamicMaterialParameters(
    Index 0 Param 1=1
    Module.Param0WriteEnabled=true
    Module.Param1WriteEnabled=false
    Module.Param2WriteEnabled=false
    Module.Param3WriteEnabled=false
   ) {enabled=false}
   /CascadeToNiagaraConverter/NiagaraScripts/ModuleScripts/AddVectorToPosition(Vector:Vector3f=&TransformVector(
     Vector=&RandomRangeVector(Minimum=<-400 -512 -32>, Maximum=<400 512 -16>, Evaluation Type=::NewEnumerator1)
     Destination Space=::Simulation
    ))
   &AddVelocity(Velocity=&RandomRangeVector(Minimum=<0 0 64>, Maximum=<0 0 128>, Evaluation Type=::NewEnumerator1))
  ]
  particleUpdateScript [
   &ParticleState()
   &SubUVAnimation(
    "Sprite Renderer "=DataInterface() {propertyValues {SpriteRenderer=&self:SmokeBlackClose_0.NiagaraSpriteRendererProperties_0}}
    Start Frame Range Override=24
    UseEndFrame=true
    UseStartFrame=true
   )
   &ScaleColor(
    Scale Alpha=&FloatFromCurve(FloatCurve=DataInterface() {propertyValues {Curve {keys [
         <at 0 0 RCIM_Cubic>
         {arriveTangent=2.6314, interpMode=RCIM_Cubic, leaveTangent=2.6314, time=0.2336, value=0.775}
         {arriveTangent=-0.1342, interpMode=RCIM_Cubic, leaveTangent=-0.1342, tangentMode=RCTM_User, time=0.3552, value=0.9347}
         {arriveTangent=-2.4933, interpMode=RCIM_Cubic, leaveTangent=-2.4933, tangentMode=RCTM_User, time=0.7701, value=0.3251}
         {arriveTangent=-0.1919, interpMode=RCIM_Cubic, leaveTangent=-0.1919, tangentMode=RCTM_User, time=1, value=-0.0111}
        ]}}})
    Scale RGB=&VectorFromCurve(VectorCurve=DataInterface() {propertyValues {
       XCurve {keys [<at 0 4.2755>,<at 0.5 0.428>,<at 1 0>]}
       YCurve {keys [<at 0 0>,<at 0.5 0.0202>,<at 1 0>]}
       ZCurve {keys [<at 0 5>,<at 0.5 0.5>,<at 1 0>]}
      }})
    ScaleA=true
    ScaleRGB=true
   )
   &DynamicMaterialParameters(
    Index 0 Param 1=1
    Index 0 Param 2=1
    Index 0 Param 3=1
    Index 0 Param 4=1
    Module.Param0WriteEnabled=true
    Module.Param1WriteEnabled=true
    Module.Param2WriteEnabled=true
    Module.Param3WriteEnabled=true
   ) {module=DynamicMaterialParameters001}
   &CameraOffset(Camera Offset Amount=25)
   &SolveForcesAndVelocity(Clamp Velocity=false, Limit Acceleration=false)
  ]
  rendererClasses [&NiagaraSpriteRendererProperties]
  renderers [{rendererClass=&NiagaraSpriteRendererProperties}]
 }}
summary {
 systemName=NS_Apollo_PowerPlant_Reactor_Steam_Only
 emitters [{emitterName=SmokeBlackClose, rendererClasses [&NiagaraSpriteRendererProperties]}]
 userVariables []
}
